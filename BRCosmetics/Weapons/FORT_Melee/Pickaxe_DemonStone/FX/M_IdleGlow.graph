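# Translucent, unlit, two-sided mesh-particle material for the idle glow on the DemonStone pickaxe.
# Inverted vertex-color blue blends two orange emissive parameters through MF_FX_EmissiveAdjustments and serves as the UV for a soft gradient mask.
# That mask is multiplied by a view-facing falloff constrained around the particle's local X axis, hiding edge-on geometry.
# Depth fade, bypassed on mobile, and particle alpha further scale the mask.
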Material /BRCosmetics/Weapons/FORT_Melee/Pickaxe_DemonStone/FX/M_IdleGlow
graph {
 properties {
  bUsedWithMeshParticles=true
  blendMode=BLEND_Translucent
  pixelDepthOffsetMode=PDOM_Legacy
  shadingModel=MSM_Unlit
  twoSided=true
  usedShadingModels=Unlit
 }
 nodes [
  n1 = VectorParameter() {parameterName=EmissiveB, defaultValue=<rgba 5 0.2075 0 0>}
  n2 = VectorParameter() {parameterName=EmissiveA, defaultValue=<rgba 5 2.3702 0 0>}
  n3 = OneMinus(None=VertexColor().B)
  MaterialFunctionCall(Emissive=Multiply(A=LinearInterpolate(A=@n1.RGB, B=@n2.RGB, Alpha=@n3)) {constB=0.2}) {materialFunction=&MF_FX_EmissiveAdjustments}
  n4 = TextureSample(UVs=@n3) {texture=/Game/Effects/Fort_Effects/Textures/Masks/T_Vert_Gradient_Softer}
  n5 = Normalize(VectorInput=Subtract(A=WorldPosition().XYZ, B=CameraPositionWS()))
  n6 = Transform(None=Constant3Vector() {constant=<rgba 1 0 0 0>}) {transformSourceType=TRANSFORMSOURCE_Local, transformType=TRANSFORM_ParticleWorld}
  n7 = CrossProduct(A=Normalize(VectorInput=CrossProduct(A=@n5, B=@n6)), B=@n6)
  n8 = Abs(None=DotProduct(A=Normalize(VectorInput=@n7), B=VertexNormalWS()))
  n9 = DepthFade() {fadeDistanceDefault=30}
  n10 = Constant() {r=1}
  Multiply(
   A=Multiply(A=Multiply(A=@n4.R, B=@n8), B=ShadingPathSwitch(Default=@n9, Mobile=@n10) {desc=Migrated from FeatureLevelSwitch})
   B=ParticleColor().A
  ) {constB=0.1}
 ]
}
