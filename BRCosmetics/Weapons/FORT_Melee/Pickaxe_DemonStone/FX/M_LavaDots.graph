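# Translucent unlit particle material for lava dot effects on the DemonStone pickaxe, drawing noise-distorted round blobs.
# Panning noise, offset per particle outside mobile, warps UVs feeding a sphere mask that drives contrast-adjusted opacity.
# The mask minus particle alpha blends two orange emissive colors, tinted by particle color.
# Outside mobile, camera proximity and depth fade shrink the mask radius instead of simply fading opacity.
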Material /BRCosmetics/Weapons/FORT_Melee/Pickaxe_DemonStone/FX/M_LavaDots
graph {
 properties {
  bUsedWithParticleSprites=true
  blendMode=BLEND_Translucent
  pixelDepthOffsetMode=PDOM_Legacy
  shadingModel=MSM_Unlit
  usedShadingModels=Unlit
 }
 nodes [
  n1 = ParticleColor()
  n2 = VectorParameter() {parameterName=EmissiveB, defaultValue=<rgba 1 0.0415 0 0>}
  n3 = VectorParameter() {parameterName=EmissiveA, defaultValue=<rgba 1 0.474 0 0>}
  n4 = VectorParameter() {parameterName=UV, defaultValue=<rgba 0.4 0.4 0.2 0>}
  n5 = ParticleRandom()
  n6 = Constant()
  n7 = Panner(
   Coordinate=Multiply(A=TextureCoordinate(), B=ComponentMask(None=@n4.RGB) {g=true, r=true})
   Time=Add(A=Time(), B=ShadingPathSwitch(Default=@n5, Mobile=@n6) {desc=Migrated from FeatureLevelSwitch})
   Speed=@n4.B
  )
  n8 = Subtract(A=ComponentMask(None=TextureSample(UVs=@n7) {samplerType=SAMPLERTYPE_LinearColor, texture=&T_NoisePack_LinearColor}.RGB) {g=true, r=true}) {constB=0.5}
  n9 = ScalarParameter() {defaultValue=0.4, parameterName=Distort}
  n10 = SphereMask(A=WorldPosition().XYZ, B=CameraPositionWS(), Radius=Constant() {r=200}, Hardness=Constant() {r=0.5})
  n11 = Multiply(A=Saturate(None=OneMinus(None=@n10)), B=DepthFade() {fadeDistanceDefault=10})
  n12 = Constant() {r=1}
  n13 = SphereMask(
   A=Add(A=TextureCoordinate(), B=Multiply(A=@n8, B=@n9))
   B=Constant() {r=0.5}
   Radius=Multiply(A=Constant() {r=0.4}, B=ShadingPathSwitch(Default=@n11, Mobile=@n12) {desc=Migrated from FeatureLevelSwitch})
   Hardness=Constant()
  )
  n14 = MaterialFunctionCall(In=Saturate(None=Subtract(A=@n13, B=@n1.A)), Contrast=ScalarParameter() {defaultValue=0.3, parameterName=Contrast2}) {materialFunction=&CheapContrast}
  MaterialFunctionCall(Emissive=Multiply(A=@n1.RGB, B=LinearInterpolate(A=@n2.RGB, B=@n3.RGB, Alpha=@n14.Result))) {materialFunction=&MF_FX_EmissiveAdjustments}
  Multiply(
   A=MaterialFunctionCall(In=@n13, Contrast=ScalarParameter() {defaultValue=2, parameterName=Contrast}) {materialFunction=&CheapContrast}.Result
   B=ScalarParameter() {defaultValue=1, parameterName=Alpha}
  )
 ]
}
