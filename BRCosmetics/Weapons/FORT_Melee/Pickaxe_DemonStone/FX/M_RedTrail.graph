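# Unlit translucent beam-trail material for the DemonStone pickaxe, drawing a red-orange glow without textures.
# A squared sphere mask around the object position, capped by inverted U, is squared again and shaped by a V sine.
# The mask blends two emissive colors, scaled by particle and dynamic alpha through MF_FX_EmissiveAdjustments.
# It also drives faint opacity with non-mobile depth fade, while CameraOffset shifts vertices along the view.
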Material /BRCosmetics/Weapons/FORT_Melee/Pickaxe_DemonStone/FX/M_RedTrail
graph {
 properties {bUsedWithBeamTrails=true, blendMode=BLEND_Translucent, pixelDepthOffsetMode=PDOM_Legacy, shadingModel=MSM_Unlit, usedShadingModels=Unlit}
 nodes [
  MaterialFunctionCall(Offset Amount=ScalarParameter() {defaultValue=5, parameterName=Offset}) {materialFunction=/Engine/Functions/Engine_MaterialFunctions02/WorldPositionOffset/CameraOffset}
  n1 = SphereMask(
   A=WorldPosition() {worldPositionShaderOffset=WPT_ExcludeAllShaderOffsets}.XYZ
   B=ObjectPositionWS()
   Radius=ScalarParameter() {defaultValue=75, parameterName=Radius}
   Hardness=Constant()
  )
  n2 = TextureCoordinate()
  n3 = Min(A=Multiply(A=@n1, B=@n1), B=OneMinus(None=ComponentMask(None=@n2) {r=true}))
  n4 = TextureCoordinate() {uTiling=0.5, vTiling=0.5}
  n5 = Multiply(A=Multiply(A=@n3, B=@n3), B=ComponentMask(None=Sine(None=@n4)) {g=true})
  n6 = LinearInterpolate(
   A=VectorParameter() {parameterName=EmissiveB, defaultValue=<rgba 1 0.0415 0 0>}.RGB
   B=VectorParameter() {parameterName=EmissiveA, defaultValue=<rgba 1 0.474 0 0>}.RGB
   Alpha=@n5
  )
  n7 = ParticleColor()
  MaterialFunctionCall(Emissive=Multiply(A=DynamicParameter() {paramNames [Alpha,Param2,Param3,Param4]}.Alpha, B=Multiply(A=@n6, B=@n7.A))) {materialFunction=&MF_FX_EmissiveAdjustments}
  Multiply(
   A=Multiply(A=@n5) {constB=0.25}
   B=ShadingPathSwitch(Default=DepthFade() {fadeDistanceDefault=15}, Mobile=Constant() {r=1}) {desc=Migrated from FeatureLevelSwitch}
  )
 ]
}
